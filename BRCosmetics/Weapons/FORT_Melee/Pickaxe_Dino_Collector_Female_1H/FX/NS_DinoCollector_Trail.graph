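# Niagara system for the weapon trail on the Dino Collector one-handed female pickaxe.
# Its single Trail emitter uses custom AnimTrail modules to spawn and initialize short-lived particles from bone or socket indices, drawn as a ribbon, within a brief delayed emitter window.
# Curves shift the ribbon from dark to pale warm beige over particle life while fading out alpha.
# It exposes SkeletalMesh and RibbonIDOffset user parameters.
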NiagaraSystem /BRCosmetics/Weapons/FORT_Melee/Pickaxe_Dino_Collector_Female_1H/FX/NS_DinoCollector_Trail
dependencies {
 usedDataInterfaces [&NiagaraDataInterfaceColorCurve, &NiagaraDataInterfaceCurve]
 usedDynamicInputs []
 usedModules [&SystemState]
}
emitters {Trail {
  emitterUpdateScript [
   &EmitterState(Life Cycle Mode=::NewEnumerator1, Loop Behavior=::NewEnumerator1, Loop Delay=0.05, Loop Duration=0.3, UseLoopDelay=true)
   &FN_SpawnAnimTrail_v2()
  ]
  particleSpawnScript [
   /Game/Effects/Niagara/Modules/AnimTrail/FN_InitializeAnimTrail/FN_InitializeAnimTrailBoneIndecies(First Bone or Socket Index=1)
   &FN_InitializeAnimTrail(Lifetime=0.25)
   &AddVelocity(Velocity=<0 1000 0>) {enabled=false}
  ]
  particleUpdateScript [
   &SolveForcesAndVelocity(Clamp Velocity=false, Limit Acceleration=false)
   &ParticleState()
   &Color(
    Color=&ColorFromCurve(ColorCurve=DataInterface() {propertyValues {
       AlphaCurve {keys [<at 0 1 RCIM_Cubic>,<at 1 1 RCIM_Cubic>]}
       BlueCurve {keys [<at 0.0075 0.0844 RCIM_Cubic>,<at 1 0.521 RCIM_Cubic>]}
       GreenCurve {keys [<at 0.0075 0.1022 RCIM_Cubic>,<at 1 0.6038 RCIM_Cubic>]}
       RedCurve {keys [<at 0.0075 0.1221 RCIM_Cubic>,<at 1 0.6795 RCIM_Cubic>]}
      }})
    Scale Alpha=&FloatFromCurve(
     FloatCurve=DataInterface() {propertyValues {Curve {keys [{interpMode=RCIM_Cubic, tangentMode=RCTM_User, time=0, value=1}, {interpMode=RCIM_Cubic, tangentMode=RCTM_User, time=1, value=0}]}}}
     Scale Curve=1.125
    )
   )
  ]
  rendererClasses [&NiagaraRibbonRendererProperties]
  renderers [{rendererClass=&NiagaraRibbonRendererProperties}]
 }}
summary {
 systemName=NS_DinoCollector_Trail
 emitters [{emitterName=Trail, rendererClasses [&NiagaraRibbonRendererProperties]}]
 userVariables [
  User.RibbonIDOffset {type=NiagaraInt32}
  User.SkeletalMesh {
   type=NiagaraDataInterfaceSkeletalMesh
   defaultValue {
    struct=NiagaraExt_VariableValue_DataInterface
    value {dataInterface=&self:NiagaraDataInterfaceSkeletalMesh_0}
   }
  }
 ]
}
